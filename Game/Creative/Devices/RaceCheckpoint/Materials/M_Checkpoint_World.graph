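# Translucent unlit material for the Creative race checkpoint device, drawing a radial ring with a center icon.
# TimeLeft drives a radial sweep that dims part of the ring and adds a bright color band, with blinking at low values.
# A noise-distorted panning gradient animates ring opacity, while a channel-masked icon is picked from a horizontal strip.
# MF_TimeTrial_Dissolve modulates both final emissive and opacity.
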Material /Game/Creative/Devices/RaceCheckpoint/Materials/M_Checkpoint_World
graph {
 properties {bEnableResponsiveAA=true, blendMode=BLEND_Translucent, pixelDepthOffsetMode=PDOM_Legacy, shadingModel=MSM_Unlit, usedShadingModels=Unlit}
 nodes [
  n1 = MaterialFunctionCall!(GradientDot) {materialFunction=/Game/Effects/Fort_Effects/Materials/UI/MF_TimeTrial_Dissolve}
  n2 = MaterialFunctionCall() {materialFunction=&VectorToRadialValue}
  n3 = ScalarParameter() {defaultValue=1, parameterName=TimeLeft}
  n4 = Add(
   A=Multiply(A=MaterialFunctionCall(UV Channel=@n2.Radial Coordinates) {materialFunction=&LinearGradient}.UGradient)
   B=LinearInterpolate(Alpha=@n3) {constA=-1.025, constB=0.05}
  )
  n5 = Clamp(None=Multiply(A=Clamp(None=@n4)) {constB=20})
  n6 = VectorParameter() {parameterName=ColorB, defaultValue=<rgba 0.0373 0.2311 0.2969 1>}
  n7 = Multiply(A=Multiply(A=@n5, B=OneMinus(None=@n5)) {constB=4}, B=@n6.RGB)
  n8 = ScalarParameter() {defaultValue=20, parameterName=BandBrightness}
  n9 = VectorParameter() {parameterName=Color, defaultValue=<rgba 0.6284 3.8929 5 1>}
  n10 = SphereMask(A=TextureCoordinate(), B=Constant() {r=0.5}, Radius=ScalarParameter() {defaultValue=0.5, parameterName=Radius}, Hardness=Constant())
  n11 = Multiply(
   A=ScalarParameter() {defaultValue=1, parameterName=Brightness}
   B=Add(A=Multiply(A=@n7, B=@n8), B=LinearInterpolate(A=@n6.RGB, B=@n9.RGB, Alpha=@n10))
  )
  Add(A=@n11, B=Multiply(A=@n11, B=@n1.Emissive))
  MaterialFunctionCall(Speed=Constant() {r=0.2}) {materialFunction=&DebugTimeSine}
  MaterialFunctionCall(Speed=Constant() {r=0.1}) {materialFunction=&DebugTimeSine}
  MaterialFunctionCall() {materialFunction=&DebugTimeSine}
  n12 = ScalarParameter() {defaultValue=1, group=Icon, parameterName=NumberOfIcons}
  n13 = AppendVector(A=@n12, B=Constant() {r=1})
  n14 = Divide(A=TextureCoordinate(), B=@n13)
  n15 = AppendVector(A=Divide(A=ScalarParameter() {group=Icon, parameterName=IconSelection}, B=@n12), B=Constant()) {desc=Min}
  n16 = Add(
   A=Multiply(A=Add(A=@n14, B=@n15), B=ScalarParameter() {defaultValue=1, group=Icon, parameterName=IconScaleUV})
   B=AppendVector(A=ScalarParameter() {group=Icon, parameterName=IconBiasU}, B=ScalarParameter() {group=Icon, parameterName=IconBiasV})
  )
  n17 = Add(A=Divide(B=@n13), B=@n15) {desc=Max}
  n18 = ChannelMaskParameter(None=TextureSampleParameter2D(UVs=Clamp(None=@n16, Min=@n15, Max=@n17)) {group=Icon, parameterName=CenterIcon, texture=&T_BlacksRGB}.RGBA) {group=Icon, parameterName=IconChannel}
  n19 = Multiply(
   A=LinearInterpolate(A=ScalarParameter() {parameterName=OpacityEdge}, B=ScalarParameter() {defaultValue=40, parameterName=OpacityCenter}, Alpha=@n10)
   B=Saturate(None=Multiply(A=@n10) {constB=40})
  )
  n20 = MaterialFunctionCall(UV Channel=MaterialFunctionCall() {materialFunction=&VectorToRadialValue}.Radial Coordinates) {materialFunction=&LinearGradient}
  n21 = Saturate(None=Add(A=Multiply(A=@n20.UGradient), B=LinearInterpolate(Alpha=@n3) {constA=-0.025, constB=1.05}))
  n22 = Multiply(
   A=Power(Base=Saturate(None=@n19), Exp=ScalarParameter() {defaultValue=1, parameterName=RingExponent})
   B=LinearInterpolate(Alpha=Floor(None=@n21)) {constA=0.05}
  )
  n23 = Multiply(A=Time(), B=ScalarParameter() {defaultValue=3, parameterName=BlinkSpeed})
  n24 = Round(None=ConstantBiasScale(None=Sine(None=@n23)))
  n25 = If(A=@n3, B=Constant() {r=0.25}, "A > B"=Constant() {r=1}, "A == B"=@n24, "A < B"=@n24)
  n26 = Panner(
   Coordinate=MaterialFunctionCall() {materialFunction=&VectorToRadialValue}.Radial Coordinates
   Speed=AppendVector(A=ScalarParameter() {parameterName=RingPanSpeedU}, B=ScalarParameter() {defaultValue=-1, parameterName=RingPanSpeedV})
  ) {bFractionalPart=true, speedY=-1}
  n27 = Panner() {speedX=0.025, speedY=0.1}
  n28 = Multiply(A=ComponentMask(None=TextureSample(UVs=@n27) {samplerType=SAMPLERTYPE_Normal, texture=/Game/Global/Textures/Default/Noise/T_Noise_Normal_01}.RGB) {g=true, r=true}) {constB=0.1}
  n29 = MaterialFunctionCall(
   In=TextureSample(UVs=Add(A=@n26, B=@n28)) {samplerSource=SSM_Wrap_WorldGroupSettings, texture=/Game/Global/Textures/Default/Masks_and_Gradients/T_Vert_Gradient_Soft}.G
   Contrast=Constant()
  ) {materialFunction=&CheapContrast}
  n30 = Saturate(None=Multiply(A=Multiply(A=@n22, B=@n25), B=LinearInterpolate(Alpha=@n29.Result) {constA=0.5}))
  Multiply(A=Saturate(None=Add(A=@n18.R, B=@n30)), B=@n1.Opacity)
 ]
}
