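# Masked unlit material for the edge of the race checkpoint ring in Creative, with a timed blink and dissolve.
# A Fresnel on the vertex normal blends Color into ColorB, scaled by Brightness, and MF_TimeTrial_Dissolve's emissive output boosts the result.
# The dissolve's opacity mask is multiplied by a gate that stays solid while TimeLeft exceeds a quarter, then switches to a rounded sine square-wave blink at BlinkSpeed.
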Material /Game/Creative/Devices/RaceCheckpoint/Materials/M_CreativeRacingRing_Edge
graph {
 properties {blendMode=BLEND_Masked, opacityMaskClipValue=0.5, pixelDepthOffsetMode=PDOM_Legacy, shadingModel=MSM_Unlit, usedShadingModels=Unlit}
 nodes [
  n1 = VertexNormalWS()
  n2 = Multiply(
   A=ScalarParameter() {defaultValue=1, parameterName=Brightness}
   B=LinearInterpolate(
    A=VectorParameter() {parameterName=Color, defaultValue=<rgba 0.1562 2.2798 3 1>}.RGB
    B=VectorParameter() {parameterName=ColorB, defaultValue=<rgba 0.025 0.0373 0.0521 1>}.RGB
    Alpha=Fresnel(Normal=@n1) {exponent=1}
   )
  )
  n3 = MaterialFunctionCall!(GradientDot=Constant2Vector() {r=1}) {materialFunction=/Game/Effects/Fort_Effects/Materials/UI/MF_TimeTrial_Dissolve}
  Add(A=@n2, B=Multiply(A=@n2, B=@n3.Emissive))
  n4 = Multiply(A=Time(), B=ScalarParameter() {defaultValue=3, parameterName=BlinkSpeed})
  n5 = Round(None=ConstantBiasScale(None=Sine(None=@n4)))
  Multiply(
   A=If(A=ScalarParameter() {defaultValue=1, parameterName=TimeLeft}, B=Constant() {r=0.25}, "A > B"=Constant() {r=1}, "A == B"=@n5, "A < B"=@n5)
   B=@n3.OpacityMask
  )
 ]
}
